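# Niagara system for the PieMan skin's body effects: dripping drool, buzzing flies and smoke, attached to skeletal mesh bones.
# Drips spawn from random limb bones as meshes stretched toward falling particles, with pendulum motion and jump or landing bursts.
# Flies swarm a bone offset through curl noise and attraction, feeding ribbon trails through particle reads.
# Smoke sprites and a distance-spawned ribbon fade with drift or movement, and stop spawning during skydiving.
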NiagaraSystem /BRCosmetics/Characters/Player/Male/Medium/Bodies/M_MED_PieMan/Effects/NS_PieMan_Skin_Original
dependencies {
 usedDataInterfaces [
  &NiagaraDataInterfaceArrayFloat
  &NiagaraDataInterfaceParticleRead
  &NiagaraDataInterfaceCurve
  &NiagaraDataInterfaceColorCurve
  /Script/Niagara.NiagaraDataInterfaceSimpleCounter
 ]
 usedDynamicInputs []
 usedModules [
  &self:SystemSpawnScript.SystemScriptSource.SystemScriptGraph.NiagaraNodeAssignment_1.SetVariables_#1
  &SystemState
  &self:SystemSpawnScript.SystemScriptSource.SystemScriptGraph.NiagaraNodeAssignment_0.SetVariables_#2
 ]
}
emitters {
 DEBUG_Sourcing {
  bEnabled=false
  emitterUpdateScript [
   &EmitterState()
   &SpawnRate(SpawnRate=6, Use Spawn Probability=false)
  ]
  particleSpawnScript [
   &InitializeParticle(Color Mode=::NewEnumerator0, Lifetime=2, Sprite Size Mode=::NewEnumerator3, UsePositionOffset=false, Write Material Random=true)
   /SetVariables(Particles.TestOffset:Vector3f=&MakeCustomVectorFromBool(Bool=true, True Vector=&RandomVector(Vector Scale=4), False Vector=<0 0 0>)) {isSetParameters=true}
   /SetVariables(Particles.PartGroupRandomInt:NiagaraInt32=&RandomRangeInt(Minimum=1, Maximum=4)) {isSetParameters=true}
   /SetVariables(
    Particles.ArmBonesL:NiagaraInt32=&self:RandomBoneIndex(RandomRange:Vector4f=<8 9 32 33>, Part Hidden Fallback:NiagaraInt32=0, Part Hidden:NiagaraBool=false)
    Particles.ArmBonesR:NiagaraInt32=&self:RandomBoneIndex(RandomRange:Vector4f=<35 36 59 60>, Part Hidden Fallback:NiagaraInt32=0, Part Hidden:NiagaraBool=false)
    Particles.LegBonesL:NiagaraInt32=&RandomRangeInt(Minimum=72, Maximum=&MakeCustomIntFromBool(Bool=$User.bisFootHidden, True INT=72, False INT=75))
    Particles.LegBonesR:NiagaraInt32=&RandomRangeInt(Minimum=65, Maximum=&MakeCustomIntFromBool(Bool=$User.bisFootHidden, True INT=65, False INT=70))
   ) {isSetParameters=true}
   /SetVariables(
    Particles.HangTime:NiagaraFloat=&RandomRangeFloat(Minimum=0.6, Maximum=0.8)
    Particles.PickBone:NiagaraInt32=&self:PickBodypart(PartGroup:NiagaraInt32=$Particles.PartGroupRandomInt)
   ) {isSetParameters=true}
   &CameraOffset(Camera Offset Amount=15)
  ]
  particleUpdateScript [
   &ParticleState()
   &SkeletalMeshLocation(
    Bone / Socket Index=11
    Bone Sampling Mode=::NewEnumerator6
    Inherit Velocity=false
    MaskSpawnGroup=false
    Sampled Position Offset Coordinate Space:ENiagaraMeshTransforms=::NewEnumerator3
    Skeletal Mesh=$User.SkeletalMesh
   )
  ]
  rendererClasses [&NiagaraSpriteRendererProperties]
  renderers [{rendererClass=&NiagaraSpriteRendererProperties}]
 }
 Drool {
  emitterSpawnScript [/Game/Effects/Niagara/Modules/Attributes/FN_Source_v2(AttributeReader=DataInterface() {propertyValues {EmitterBinding {bindingMode=Other, emitterName=DroolSource}}})]
  emitterUpdateScript [&EmitterState()]
  particleSpawnScript [/Niagara/Modules/Spawn/Initialization/InitializeParticle(
    Color=&RandomRangeLinearColor(Minimum=<rgba 1 0.5793 0.463 1>, Maximum=<rgba 1 0.8891 0.792 1>)
    Lifetime=&UniformRangedFloat(Minimum=1.5, Maximum=1.5)
    Mesh Scale=<0.25 0.25 0.015>
    Module.Write Color=true
    Module.Write Lifetime=true
    Module.Write Mass=true
    Module.Write Material Random=true
    Module.Write Position=false
    Module.Write Scale=true
    Module.Write SpriteRotation=false
    Module.Write SpriteSize=true
    Sprite Size=&Vector2DFromFloat(Value=5)
   )]
  particleUpdateScript [
   &ParticleState()
   &self:UpdateParentPosition() {module=UpdateParentPosition}
   /SetVariables(
    Particles.XYScale:NiagaraFloat=&FloatFromCurve(FloatCurve=DataInterface() {propertyValues {Curve {keys [
         {arriveTangent=10.1742, interpMode=RCIM_Cubic, leaveTangent=10.1742, tangentMode=RCTM_User, time=0, value=0}
         {arriveTangent=-5.3219, interpMode=RCIM_Cubic, leaveTangent=-5.3219, tangentMode=RCTM_User, time=0.1, value=1}
         {arriveTangent=1.3575, interpMode=RCIM_Cubic, leaveTangent=1.3576, tangentMode=RCTM_User, time=0.4, value=0.5}
         {arriveTangent=-0.5455, interpMode=RCIM_Cubic, leaveTangent=-0.5455, time=0.65, value=0.7182}
         {arriveTangent=-3.572, interpMode=RCIM_Cubic, leaveTangent=-3.572, tangentMode=RCTM_User, time=0.95, value=0.2}
         <at 1 0.05>
        ]}}})
    Particles.ZScale:NiagaraFloat=&FloatFromCurve(
     FloatCurve=DataInterface() {propertyValues {Curve {keys [
         {arriveTangent=3.0916, interpMode=RCIM_Cubic, leaveTangent=3.0916, tangentMode=RCTM_User, time=0, value=0}
         {arriveTangent=-1.9901, interpMode=RCIM_Cubic, leaveTangent=-1.9901, tangentMode=RCTM_User, time=0.296, value=1}
         {arriveTangent=1.4204, interpMode=RCIM_Cubic, leaveTangent=1.4204, tangentMode=RCTM_User, time=0.8, value=0.4081}
         {arriveTangent=4.145, interpMode=RCIM_Cubic, leaveTangent=4.145, time=0.95, value=1}
         <at 0.9984 1.2306>
        ]}}}
     Scale Curve=&RandomRangeFloat(Minimum=15, Maximum=15)
    )
   ) {isSetParameters=true}
   &PendulumSetup(
    Module.Enable Spring Driver=true
    Pendulum Length:NiagaraFloat=$Particles.ZScale
    Pendulum Pivot:Vector3f=$Particles.ParentPosition
    Rigid Pendulum=false
    "Spring Driven Constraint (Tightness)":NiagaraFloat=&FloatFromCurve(FloatCurve=DataInterface() {propertyValues {Curve {keys [<at 0 1>,<at 0.2 0>]}}}, Scale Curve=15)
   )
   &Drag(Drag=-0.5, Use Linear Drag=true, Use Rotational Drag=false)
   &AddVelocity(Velocity=<0 0 -25>)
   &InheritVelocity(Module.Limit Speed=true, Velocity Scale=&Lerp_Vector(A=<-1 -1 -1>, B=<0 0 0>, Alpha=$Particles.NormalizedAge))
   &SolveForcesAndVelocity(Clamp Velocity=false, Limit Acceleration=false)
   &PendulumConstraint()
   &CalculateAccurateVelocity()
   /SetVariables(Particles.MeshFacing:Vector3f=&NormalizeVector(Vector To Normalize=&Subtract_Vector(A=$Particles.ParentPosition, B=Linked() {linkedVariable {name=Particles.Position, type=NiagaraPosition}}))) {isSetParameters=true, module=SetVariables_#3}
   /Niagara/Modules/Update/Orientation/V2/OrientMeshToVector(
    Look At Direction:Vector3f=$Particles.MeshFacing
    Override Quat=false
    Reference Axis Vector:Vector3f=<0 0 1>
    UseOptionalSideAxis=false
    UseOptionalUpAxis=false
   )
   &DynamicMaterialParameters(
    Index 2 Param 1=HlslExpression() {hlslExpression=Particles.ParentPosition.x}
    Index 2 Param 2=HlslExpression() {hlslExpression=Particles.ParentPosition.y}
    Index 2 Param 3=HlslExpression() {hlslExpression=Particles.ParentPosition.z}
    Index 2 Param 4=2
    Index 3 Param 1=&MakeFloatFromVector(Vector=Linked() {linkedVariable {name=Particles.Position, type=NiagaraPosition}}, Channel=::NewEnumerator0)
    Index 3 Param 2=&MakeFloatFromVector(Vector=Linked() {linkedVariable {name=Particles.Position, type=NiagaraPosition}}, Channel=::NewEnumerator1)
    Index 3 Param 3=&MakeFloatFromVector(Vector=Linked() {linkedVariable {name=Particles.Position, type=NiagaraPosition}}, Channel=::NewEnumerator2)
    Index 3 Param 4=0.2
    Module.Param0WriteEnabled2=true
    Module.Param0WriteEnabled3=true
    Module.Param1WriteEnabled2=true
    Module.Param1WriteEnabled3=true
    Module.Param2WriteEnabled2=true
    Module.Param2WriteEnabled3=true
    Module.Param3WriteEnabled2=true
    Module.Param3WriteEnabled3=true
    Write Parameter Index 0=false
    Write Parameter Index 2=true
    Write Parameter Index 3=true
   )
   &DynamicMaterialParameters(
    Index 2 Param 1=HlslExpression() {hlslExpression=Particles.ParentPosition.x}
    Index 2 Param 2=HlslExpression() {hlslExpression=Particles.ParentPosition.y}
    Index 2 Param 3=HlslExpression() {hlslExpression=Particles.ParentPosition.z}
    Index 3 Param 1=&MakeFloatFromVector(Vector=Linked() {linkedVariable {name=Particles.Position, type=NiagaraPosition}}, Channel=::NewEnumerator0)
    Index 3 Param 2=&MakeFloatFromVector(Vector=Linked() {linkedVariable {name=Particles.Position, type=NiagaraPosition}}, Channel=::NewEnumerator1)
    Index 3 Param 3=&MakeFloatFromVector(Vector=Linked() {linkedVariable {name=Particles.Position, type=NiagaraPosition}}, Channel=::NewEnumerator2)
    Module.Param0WriteEnabled2=true
    Module.Param0WriteEnabled3=true
    Module.Param1WriteEnabled2=true
    Module.Param1WriteEnabled3=true
    Module.Param2WriteEnabled2=true
    Module.Param2WriteEnabled3=true
    Module.Param3WriteEnabled2=true
    Module.Param3WriteEnabled3=true
    Write Parameter Index 0=false
    Write Parameter Index 2=true
    Write Parameter Index 3=true
   ) {enabled=false, module=DynamicMaterialParameters001}
   &ScaleMeshSize(Scale Factor=&MakeVector(
     X=$Particles.XYScale
     Y=$Particles.XYScale
     Z=&VectorLength(Vector=&Subtract_Vector(A=$Particles.ParentPosition, B=Linked() {linkedVariable {name=Particles.Position, type=NiagaraPosition}}))
    ))
   &CameraOffset(Camera Offset Amount=5)
  ]
  rendererClasses [&NiagaraMeshRendererProperties, &NiagaraSpriteRendererProperties]
  renderers [
   {rendererClass=&NiagaraMeshRendererProperties, rendererIndex=0}
   {rendererClass=&NiagaraSpriteRendererProperties, rendererIndex=1}
  ]
 }
 DroolSimple {
  emitterUpdateScript [
   &EmitterState(Life Cycle Mode=::NewEnumerator1, Loop Duration=3.5, UseLoopDelay=false)
   &SpawnRate(Spawn Probability=0.25, SpawnRate=4, Use Spawn Probability=true)
   &SpawnRate(
    Spawn Probability=&MakeCustomFloatFromBool(Bool=$User.bIsJumping, True Float=1, False Float=0.25)
    SpawnRate=&MakeCustomFloatFromBool(Bool=$User.bIsJumping, True Float=150, False Float=4)
    Use Spawn Probability=true
   ) {module=SpawnRate002}
   &SpawnRate(
    SpawnRate=&MakeCustomFloatFromBool(Bool=$User.SpawnLandedFX, True Float=&RandomRangeFloat(Minimum=100, Maximum=200), False Float=0)
    Use Spawn Probability=false
   ) {module=SpawnRate001}
  ]
  particleSpawnScript [
   /SetVariables(
    Particles.ArmBonesL:NiagaraInt32=&self:RandomBoneIndex(RandomRange:Vector4f=<8 9 32 33>, Part Hidden Fallback:NiagaraInt32=7, Part Hidden:NiagaraBool=$User.bisHandHidden)
    Particles.ArmBonesR:NiagaraInt32=&self:RandomBoneIndex(RandomRange:Vector4f=<35 36 59 60>, Part Hidden Fallback:NiagaraInt32=0, Part Hidden:NiagaraBool=false)
    Particles.LegBonesL:NiagaraInt32=&RandomRangeInt(Minimum=71, Maximum=&MakeCustomIntFromBool(Bool=$User.bisFootHidden, True INT=71, False INT=74))
    Particles.LegBonesR:NiagaraInt32=&RandomRangeInt(Minimum=64, Maximum=&MakeCustomIntFromBool(Bool=$User.bisFootHidden, True INT=64, False INT=67))
   ) {isSetParameters=true}
   /SetVariables(
    Particles.HangTime:NiagaraFloat=&RandomRangeFloat(Minimum=0.6, Maximum=0.8)
    Particles.PendulumLength:NiagaraFloat=&RandomRangeFloat(Minimum=7, Maximum=20)
    Particles.PickBone:NiagaraInt32=&self:PickBodypart(PartGroup:NiagaraInt32=&MakeCustomIntFromBool(
      Bool=$User.bisFootHidden
      True INT=&RandomRangeInt(Minimum=1, Maximum=2)
      False INT=&RandomRangeInt(Minimum=1, Maximum=4)
     ))
    Particles.SpawnedOnJumpOrLand:NiagaraBool=&Boolean_OrOperation(Bool1=&SetBoolByFloatComparison(A=$User.JumpTime, B=0), Bool2=$User.SpawnLandedFX)
   ) {isSetParameters=true}
   &InitializeParticle(
    Lifetime=&Power(a=&UniformRangedFloat(Minimum=0.5, Maximum=1.2), b=2)
    Mesh Scale=&Multiply_VectorByFloat(
     Vector=<0.225 0.225 0.01>
     Float=&MakeCustomFloatFromBool(
      Bool=$Particles.SpawnedOnJumpOrLand
      True Float=&RandomRangeFloat(Minimum=0.75, Maximum=1.5)
      False Float=&RandomRangeFloat(Minimum=0.75, Maximum=1)
     )
    )
    Module.Write Color=false
    Module.Write Lifetime=true
    Module.Write Mass=true
    Module.Write Material Random=true
    Module.Write Position=false
    Module.Write Scale=true
    Module.Write SpriteRotation=false
    Module.Write SpriteSize=true
    Sprite Size=&Vector2DFromFloat(Value=2)
   )
   &AddVelocity(Velocity=&MakeCustomVectorFromBool(Bool=$Particles.SpawnedOnJumpOrLand, True Vector=&RandomVector(Vector Scale=50), False Vector=<0 0 0>))
   &SkeletalMeshLocation(
    Bone / Socket Index=$Particles.PickBone
    Bone Sampling Mode=::NewEnumerator6
    Inherit Velocity=false
    MaskSpawnGroup=false
    Sampled Position Offset=&MakeCustomVectorFromBool(
     Bool=$Particles.SpawnedOnJumpOrLand
     True Vector=&Multiply_VectorByFloat(Vector=$Particles.Velocity, Float=0.2)
     False Vector=<0 0 0>
    )
    Skeletal Mesh=$User.SkeletalMesh
   ) {module=SkeletalMeshLocation001}
  ]
  particleUpdateScript [
   &ParticleState()
   &SkeletalMeshLocation(
    Apply To Particle Position=false
    Bone / Socket Index=$Particles.PickBone
    Bone Sampling Mode=::NewEnumerator6
    Inherit Velocity=false
    MaskSpawnGroup=false
    Skeletal Mesh=$User.SkeletalMesh
   )
   /SetVariables(
    Particles.CompareZ:NiagaraFloat=&OneMinusFloat(Float=&ClampFloat(Float=&Multiply_Float(
       A=&Subtract_Float(
        A=&MakeFloatFromVector(Vector=Linked() {linkedVariable {name=Particles.Position, type=NiagaraPosition}}, Channel=::NewEnumerator2)
        B=&MakeFloatFromVector(Vector=$Output.SkeletalMeshLocation.SampledPosition, Channel=::NewEnumerator2)
       )
       B=0.01
      )))
    Particles.MeshPosition:Vector3f=&Lerp_Vector(
     A=$Output.SkeletalMeshLocation.SampledPosition
     B=Linked() {linkedVariable {name=Particles.Position, type=NiagaraPosition}}
     Alpha=&Add_Float(A=0.9, B=&Multiply_Float(A=0.1, B=$Particles.NormalizedAge))
    )
   ) {isSetParameters=true}
   &GravityForce()
   &SolveForcesAndVelocity(Clamp Velocity=false, Limit Acceleration=false)
   &CalculateAccurateVelocity()
   /SetVariables(Particles.MeshFacing:Vector3f=&NormalizeVector(Vector To Normalize=&Subtract_Vector(A=$Output.SkeletalMeshLocation.SampledPosition, B=Linked() {linkedVariable {name=Particles.Position, type=NiagaraPosition}}))) {isSetParameters=true, module=SetVariables_#3}
   &OrientMeshToVector(
    Look At Direction:Vector3f=&Multiply_VectorByFloat(Vector=$Particles.MeshFacing, Float=1)
    Override Quat=false
    Reference Axis Vector:Vector3f=<0 0 1>
    UseOptionalSideAxis=false
    UseOptionalUpAxis=false
   )
   &DynamicMaterialParameters(
    Index 2 Param 1=HlslExpression() {hlslExpression=Particles.MeshPosition.x}
    Index 2 Param 2=HlslExpression() {hlslExpression=Particles.MeshPosition.y}
    Index 2 Param 3=HlslExpression() {hlslExpression=Particles.MeshPosition.z}
    Index 2 Param 4=0.5
    Index 3 Param 1=&MakeFloatFromVector(Vector=Linked() {linkedVariable {name=Particles.Position, type=NiagaraPosition}}, Channel=::NewEnumerator0)
    Index 3 Param 2=&MakeFloatFromVector(Vector=Linked() {linkedVariable {name=Particles.Position, type=NiagaraPosition}}, Channel=::NewEnumerator1)
    Index 3 Param 3=&MakeFloatFromVector(Vector=Linked() {linkedVariable {name=Particles.Position, type=NiagaraPosition}}, Channel=::NewEnumerator2)
    Index 3 Param 4=0.2
    Module.Param0WriteEnabled2=true
    Module.Param0WriteEnabled3=true
    Module.Param1WriteEnabled2=true
    Module.Param1WriteEnabled3=true
    Module.Param2WriteEnabled2=true
    Module.Param2WriteEnabled3=true
    Module.Param3WriteEnabled2=true
    Module.Param3WriteEnabled3=true
    Write Parameter Index 0=false
    Write Parameter Index 2=true
    Write Parameter Index 3=true
   )
   /SetVariables(Particles.XYScale:NiagaraFloat=&FloatFromCurve(FloatCurve=DataInterface() {propertyValues {Curve {keys [
         {arriveTangent=-3.383, interpMode=RCIM_Cubic, leaveTangent=-3.383, tangentMode=RCTM_User, time=0, value=1}
         {arriveTangent=-0.5412, interpMode=RCIM_Cubic, leaveTangent=-0.5412, tangentMode=RCTM_User, time=1, value=0}
        ]}}})) {isSetParameters=true}
   &ScaleMeshSize(Scale Factor=&Multiply_VectorByFloat(
     Vector=&MakeVector(
      X=$Particles.XYScale
      Y=$Particles.XYScale
      Z=&ClampFloat(
       Float=&VectorLength(Vector=&Subtract_Vector(A=$Particles.MeshPosition, B=Linked() {linkedVariable {name=Particles.Position, type=NiagaraPosition}}))
       Max=$Particles.PendulumLength
      )
     )
     Float=$Particles.CompareZ
    )) {module=ScaleMeshSize001}
   &Color(Color=&ColorFromCurve(
     ColorCurve=DataInterface() {propertyValues {
       AlphaCurve {keys [<at 0 20 RCIM_Cubic>]}
       BlueCurve {keys [<at 0 0 RCIM_Cubic>,<at 0.5257 0>,<at 1 5 RCIM_Cubic>]}
       GreenCurve {keys [<at 0 0 RCIM_Cubic>,<at 0.5257 5>,<at 1 0 RCIM_Cubic>]}
       RedCurve {keys [<at 0 5 RCIM_Cubic>,<at 0.5257 0>,<at 1 0 RCIM_Cubic>]}
      }}
     CurveIndex=&VectorLength(Vector=$Particles. PositionCompare)
    )) {enabled=false}
  ]
  rendererClasses [&NiagaraSpriteRendererProperties, &NiagaraMeshRendererProperties]
  renderers [
   {rendererClass=&NiagaraSpriteRendererProperties}
   {rendererClass=&NiagaraMeshRendererProperties, rendererIndex=1}
   {rendererClass=&NiagaraSpriteRendererProperties, rendererIndex=2}
  ]
 }
 DroolSource {
  emitterUpdateScript [
   &EmitterState()
   &SpawnRate(Spawn Probability=0.25, SpawnRate=6, Use Spawn Probability=true)
  ]
  particleSpawnScript [
   &InitializeParticle(
    Lifetime=2
    Module.Write Color=true
    Module.Write Lifetime=true
    Module.Write Mass=false
    Module.Write Material Random=true
    Module.Write Position=false
    Module.Write Scale=false
    Module.Write SpriteRotation=false
    Module.Write SpriteSize=true
    Sprite Size=&Vector2DFromFloat(Value=5)
   )
   /SetVariables(Particles.TestOffset:Vector3f=&MakeCustomVectorFromBool(Bool=true, True Vector=&RandomVector(Vector Scale=3), False Vector=<0 0 0>)) {isSetParameters=true}
   /SetVariables(Particles.PartGroupRandomInt:NiagaraInt32=&RandomRangeInt(Minimum=1, Maximum=4)) {isSetParameters=true}
   /SetVariables(
    Particles.ArmBonesL:NiagaraInt32=&self:RandomBoneIndex(RandomRange:Vector4f=<8 9 32 33>, Part Hidden Fallback:NiagaraInt32=7, Part Hidden:NiagaraBool=$User.bisHandHidden)
    Particles.ArmBonesR:NiagaraInt32=&self:RandomBoneIndex(RandomRange:Vector4f=<35 36 59 60>, Part Hidden Fallback:NiagaraInt32=34, Part Hidden:NiagaraBool=$User.bisHandHidden)
    Particles.LegBonesL:NiagaraInt32=&RandomRangeInt(Minimum=71, Maximum=&MakeCustomIntFromBool(Bool=$User.bisFootHidden, True INT=71, False INT=74))
    Particles.LegBonesR:NiagaraInt32=&RandomRangeInt(Minimum=64, Maximum=&MakeCustomIntFromBool(Bool=$User.bisFootHidden, True INT=64, False INT=67))
   ) {isSetParameters=true}
   /SetVariables(
    Particles.HangTime:NiagaraFloat=&RandomRangeFloat(Minimum=0.6, Maximum=0.8)
    Particles.PickBone:NiagaraInt32=&self:PickBodypart(PartGroup:NiagaraInt32=$Particles.PartGroupRandomInt)
   ) {isSetParameters=true}
   &SkeletalMeshLocation(
    Bone / Socket Index=$Particles.PickBone
    Bone Sampling Mode=::NewEnumerator6
    Inherit Velocity=false
    MaskSpawnGroup=false
    Sampled Position Offset=$Particles.TestOffset
    Skeletal Mesh=$User.SkeletalMesh
   ) {enabled=false, module=SkeletalMeshLocation001}
   &ApplyInitialForces()
   &RegionMask(Region Type:ENiagaraDebugDrawShape=::NewEnumerator1, UseOffset=false, UseRotation=false)
   &GenerateLocationEvent(Float to Send as Random Float:NiagaraFloat=$Particles.HangTime, Module.Use Event Send Rate=true)
  ]
  particleUpdateScript [
   &ParticleState()
   &SkeletalMeshLocation(
    Apply To Particle Position:NiagaraBool=&SetBoolByFloatComparison(A=$Particles.NormalizedAge, B=$Particles.HangTime, Comparison Type=::NewEnumerator6)
    Bone / Socket Index=$Particles.PickBone
    Bone Sampling Mode=::NewEnumerator6
    Inherit Velocity=false
    MaskSpawnGroup=false
    Sampled Position Offset=&Add_Vector(
     A=$Particles.TestOffset
     B=&MakeVector(
      X=&SelectFloatFromArray(Direct Array Index=$Particles.PartGroupRandomInt, Float Selection Array=DataInterface() {propertyValues {FloatData [0,15,-15,0,0]}})
      Y=0
      Z=0
     )
    )
    Sampled Position Offset Coordinate Space:ENiagaraMeshTransforms=::NewEnumerator3
    Skeletal Mesh=$System.Skeletal Mesh
   )
   &AddVelocity(Velocity=&MakeCustomVectorFromBool(
     Bool=&SetBoolByFloatComparison(A=$Particles.NormalizedAge, B=$Particles.HangTime, Comparison Type=::NewEnumerator2)
     True Vector=<0 0 -50>
     False Vector=<0 0 0>
    ))
   &SolveForcesAndVelocity(Clamp Velocity=false, Limit Acceleration=false)
   &KillParticles()
  ]
  rendererClasses [&NiagaraSpriteRendererProperties]
  renderers [{rendererClass=&NiagaraSpriteRendererProperties}]
 }
 Flies_Optimized {
  emitterUpdateScript [
   &EmitterState()
   &SpawnRate(
    Spawn Probability=&Lerp_Float(A=0.5, B=0, Alpha=&MakeFloatFromInt(Int=$System.isSkydiving))
    SpawnRate=5
    Use Spawn Probability=true
   )
  ]
  particleSpawnScript [
   &InitializeParticle(
    Color=&RandomRangeLinearColor(Minimum=<rgba 0 0.0057 0.057 1>, Maximum=<rgba 0.005 0.02 0 1>)
    Lifetime=&Lerp_Float(A=1, B=3, Alpha=&Power(a=&RandomRangeFloat(Minimum=0, Maximum=1), b=4))
    Mesh Scale=&Multiply_VectorByFloat(Vector=<1 2 1>, Float=&UniformRangedFloat(Minimum=0.1, Maximum=0.15))
    Module.Write Color=true
    Module.Write Lifetime=true
    Module.Write Mass=false
    Module.Write Material Random=true
    Module.Write Position=false
    Module.Write Scale=true
    Module.Write SpriteRotation=false
    Module.Write SpriteSize=false
    Position=&SimulationPosition()
   )
   &DynamicMaterialParameters(
    Index 0 Param 1=10
    Index 0 Param 2=1
    Index 1 Param 1=&UniformRangedFloat(Minimum=0, Maximum=30)
    Index 1 Param 2=&UniformRangedFloat(Minimum=0, Maximum=1)
    Index 1 Param 3=&UniformRangedFloat(Minimum=0, Maximum=1)
    Index 1 Param 4=&UniformRangedFloat(Minimum=0, Maximum=1)
    Module.Param0WriteEnabled=true
    Module.Param0WriteEnabled1=true
    Module.Param1WriteEnabled=true
    Module.Param1WriteEnabled1=false
    Module.Param2WriteEnabled=true
    Module.Param2WriteEnabled1=false
    Module.Param3WriteEnabled=true
    Module.Param3WriteEnabled1=false
    Write Parameter Index 1=true
   )
   &SkeletalMeshLocation(
    Apply To Particle Position=false
    Bone / Socket Index=2
    Bone Sampling Mode=::NewEnumerator4
    Inherit Velocity=false
    MaskSpawnGroup=false
    Sampled Position Offset=&Add_Vector(
     A=<-22.1231 -11 -0.8408>
     B=&RotateVectorByQuaternion(Vector To Rotate=&RandomRangeVector(Minimum=<0 16.1011 0>, Maximum=<30 0 2>), Quaternion=<0 -0.15 0 1>)
    )
    Sampled Position Offset Coordinate Space:ENiagaraMeshTransforms=::NewEnumerator3
    Skeletal Mesh=$System.Skeletal Mesh
   )
   /Niagara/Modules/Spawn/Location/SphereLocation(
    Hemisphere Z=true
    MaskSpawnGroup=false
    Module.Override Randomness=false
    Sphere Origin:Vector3f=$Output.SkeletalMeshLocation.SampledPosition
    Sphere Radius=20
    Surface Only=false
   )
  ]
  particleUpdateScript [
   &ParticleState()
   &SkeletalMeshLocation(
    Apply To Particle Position=false
    Bone / Socket Index=2
    Bone Sampling Mode=::NewEnumerator4
    Inherit Velocity=false
    MaskSpawnGroup=false
    Sampled Position Offset=&Add_Vector(
     A=<-22.1231 -11 -0.8408>
     B=&RotateVectorByQuaternion(Vector To Rotate=&RandomRangeVector(Minimum=<0 16.1011 0>, Maximum=<30 0 2>), Quaternion=<0 -0.15 0 1>)
    )
    Sampled Position Offset Coordinate Space:ENiagaraMeshTransforms=::NewEnumerator3
    Skeletal Mesh=$System.Skeletal Mesh
   ) {module=SkeletalMeshLocation001}
   /SetVariables(Particles.SourceDistance:NiagaraFloat=/Niagara/DynamicInputs/TypeConversions/VectorLengthSafe(Vector:Vector3f=&Subtract_Vector(A=$Output.SkeletalMeshLocation001.SampledPosition, B=Linked() {linkedVariable {name=Particles.Position, type=NiagaraPosition}}))) {isSetParameters=true}
   /Niagara/Modules/Update/Forces/CurlNoiseForce(
    Noise Frequency:NiagaraFloat=&RandomRangeFloat(Minimum=0.05, Maximum=0.25)
    Noise Strength=1000
    Pan Noise=true
    Pan Noise Field=<0 0 1>
   )
   /Niagara/Modules/Update/Forces/PointAttractionForce(
    Attraction Radius=1000
    AttractionStrength:NiagaraFloat=&Lerp_Float(A=0, B=1, Alpha=$Particles.SourceDistance)
    AttractorPosition:Vector3f=$Output.SkeletalMeshLocation001.SampledPosition
    Module.Kill Within Radius=false
    Module.Use Falloff=true
   ) {module=PointAttractionForce002}
   &Drag(Use Linear Drag=true, Use Rotational Drag=false)
   &DynamicMaterialParameters(
    Index 0 Param 1=&UniformRangedFloat(Minimum=4, Maximum=5)
    Index 0 Param 2=&FloatFromCurve(FloatCurve=DataInterface() {propertyValues {Curve {keys [
         {interpMode=RCIM_Cubic, tangentMode=RCTM_User, time=0, value=0}
         {arriveTangent=0.002, interpMode=RCIM_Cubic, leaveTangent=0.002, tangentMode=RCTM_User, time=0.1827, value=1}
         {arriveTangent=-1.2195, interpMode=RCIM_Cubic, leaveTangent=-1.2195, tangentMode=RCTM_User, time=0.9833, value=0.6448}
         <at 1.0011 0.002>
        ]}}})
    Index 1 Param 1=&UniformRangedFloat(Minimum=0, Maximum=30)
    Index 1 Param 2=1.1418
    Index 1 Param 3=&UniformRangedFloat(Minimum=0, Maximum=1)
    Index 1 Param 4=&UniformRangedFloat(Minimum=0, Maximum=1)
    Module.Param0WriteEnabled=false
    Module.Param0WriteEnabled1=true
    Module.Param1WriteEnabled=true
    Module.Param1WriteEnabled1=false
    Module.Param2WriteEnabled=false
    Module.Param2WriteEnabled1=true
    Module.Param3WriteEnabled=false
    Module.Param3WriteEnabled1=true
    Write Parameter Index 1=true
   ) {module=DynamicMaterialParameters001}
   &SolveForcesAndVelocity(Clamp Velocity=false, Limit Acceleration=false)
   &ScaleMeshSize(Scale Factor=&VectorFromFloat(Value=&FloatFromCurve(
      FloatCurve=DataInterface() {propertyValues {Curve {keys [
          {interpMode=RCIM_Cubic, tangentMode=RCTM_User, time=-0.0046, value=1.0113}
          {arriveTangent=-1.0067, interpMode=RCIM_Cubic, leaveTangent=-1.0067, tangentMode=RCTM_User, time=0.9283, value=0.7392}
          <at 1 0>
         ]}}}
      Scale Curve=2
     )))
   &KillParticlesInVolume(Invert Volume=true, Origin Offset:Vector3f=<0 0 100>, Sphere Radius=200)
  ]
  rendererClasses [&NiagaraMeshRendererProperties]
  renderers [{rendererClass=&NiagaraMeshRendererProperties, rendererIndex=0}]
 }
 Fly_Trail {
  emitterUpdateScript [
   &EmitterState(Enable Distance Culling=true, MaxDistance=2500, Module.bUseMaxDistance=true, Module.bUseMinDistance=false, Scalability Mode=::NewEnumerator1)
   /SetVariables(Emitter.EmitterReader=DataInterface() {propertyValues {EmitterBinding {bindingMode=Other, emitterName=Flies_Optimized}}}) {isSetParameters=true}
   &SpawnRate(
    SpawnRate=/Game/Effects/Niagara/DynamicInputs/Attributes/FN_GetEmitterParticleCount(AttributeReader:NiagaraDataInterfaceParticleRead=$Emitter.EmitterReader, Multiplier:NiagaraFloat=20)
    Use Spawn Probability=true
   )
  ]
  particleSpawnScript [
   &self:Sample_Location(New Particle Attribute Reader:NiagaraDataInterfaceParticleRead=$Emitter.EmitterReader) {module=Sample_Location}
   &InitializeRibbon(
    Color=<rgba 1 1 1 0.4>
    Lifetime=0.5
    Module.Write Color=true
    Module.Write Lifetime=true
    Module.Write Mass=true
    Module.Write Material Random=true
    Module.Write Position=false
    Module.Write RibbonTwist=false
    Module.Write RibbonWidth=true
    Ribbon Width=&RandomRangeFloat(Minimum=2, Maximum=6)
   )
  ]
  particleUpdateScript [
   &ParticleState()
   &SolveForcesAndVelocity(Clamp Velocity=false, Limit Acceleration=false)
   /SetVariables(Particles.DistanceFromPlayer:NiagaraFloat=&Multiply_Float(
     A=&Add_Float(
      A=&VectorLength(Vector=&Multiply_Vector(
        A=&Subtract_Vector(
         A=Linked() {linkedVariable {name=Particles.Position, type=NiagaraPosition}}
         B=&Add_Vector(A=Linked() {linkedVariable {name=Engine.Owner.Position, type=NiagaraPosition}}, B=<0 0 100>)
        )
        B=<1.1 1.1 0.5>
       ))
      B=-75
     )
     B=0.015
    )) {isSetParameters=true}
   &ScaleColor(
    Scale Alpha=/Niagara/DynamicInputs/Math/LerpMultipleFloats(
     Float Count:ENiagara_MultipleLerpCount=::NewEnumerator2
     1:NiagaraFloat=0
     2:NiagaraFloat=1
     3:NiagaraFloat=1
     4:NiagaraFloat=0
     5:NiagaraFloat=0
     Alpha:NiagaraFloat=$Particles.NormalizedAge
    )
    ScaleA=true
    ScaleRGB=false
   )
   &ScaleColor(
    Color Value To Scale=$Particles.Color
    Scale Alpha=&FN_FrontEndFloatModifier(Gameplay=&Lerp_Float(A=1, B=0, Alpha=$Particles.DistanceFromPlayer), Front End=1)
    ScaleA=true
    ScaleRGB=false
   ) {module=ScaleColor001}
   &ScaleRibbonWidth(Ribbon Width Scale=&FloatFromCurve(
     FloatCurve=DataInterface() {propertyValues {Curve {keys [<at 0 1 RCIM_Cubic>,<at 1 0>]}}}
     Scale Curve=&VectorLength(Vector=$Particles.InheritedScale)
    ))
   &Color(Color=&ColorFromCurve(
     ColorCurve=DataInterface() {propertyValues {
       AlphaCurve {keys [<at 0 1 RCIM_Cubic>]}
       BlueCurve {keys [<at 0 0 RCIM_Cubic>,<at 0.3333 0>,<at 0.6667 0>,<at 1 1 RCIM_Cubic>]}
       GreenCurve {keys [<at 0 0 RCIM_Cubic>,<at 0.3333 1>,<at 0.6667 1>,<at 1 0 RCIM_Cubic>]}
       RedCurve {keys [<at 0 1 RCIM_Cubic>,<at 0.3333 1>,<at 0.6667 0>,<at 1 0 RCIM_Cubic>]}
      }}
     CurveIndex=$Particles.DistanceFromPlayer
    )) {enabled=false}
  ]
  rendererClasses [&NiagaraRibbonRendererProperties]
  renderers [{rendererClass=&NiagaraRibbonRendererProperties}]
 }
 Smoke {
  emitterUpdateScript [
   &EmitterState()
   &SpawnRate(
    SpawnRate=&Multiply_Float(
     A=80
     B=&Lerp_Float(
      A=1
      B=0.5
      Alpha=&ClampFloat(Float=&Multiply_Float(A=0.1, B=&VectorLengthSafe(Vector:Vector3f=$Engine.Owner.Velocity)))
     )
    )
    Use Spawn Probability=false
   ) {enabled=false}
   &SpawnRate(
    Spawn Probability=&Lerp_Float(A=1, B=0, Alpha=&MakeFloatFromInt(Int=$System.isSkydiving))
    SpawnRate=50
    Use Spawn Probability=true
   ) {module=SpawnRate001}
   &SpawnRate(
    Spawn Probability=&Lerp_Float(A=1, B=0, Alpha=&MakeFloatFromInt(Int=$System.isSkydiving))
    SpawnGroup=1
    SpawnRate=25
    Use Spawn Probability=true
   ) {module=SpawnRate002}
  ]
  particleSpawnScript [
   &InitializeParticle(
    Color=<rgba 1 1 1 0.2>
    Lifetime=&RandomRangeFloat(Minimum=1, Maximum=2)
    Module.Write Color=true
    Module.Write Lifetime=true
    Module.Write Mass=false
    Module.Write Material Random=true
    Module.Write Position=false
    Module.Write Scale=false
    Module.Write SpriteRotation=true
    Module.Write SpriteSize=true
    Position=&SimulationPosition()
    Sprite Rotation=&UniformRangedFloat(Minimum=0, Maximum=360)
    Sprite Size=&Vector2DFromFloat(Value=&UniformRangedFloat(Minimum=5, Maximum=17))
   )
   &AccelerationForce(Acceleration=&UniformRangedVector(Minimum=<0 0 0>, Maximum=<0 0 50>))
   &ApplyInitialForces()
   /SetVariables(Particles.ShoulderSmokeOffset:Vector3f=&RandomRangeVector(Minimum=<0 -5 20>, Maximum=<0 -5 -20>)) {isSetParameters=true}
   &SkeletalMeshLocation(
    Bone / Socket Index=2
    Bone Sampling Mode=::NewEnumerator4
    Inherit Velocity=false
    MaskSpawnGroup=true
    Sampled Position Offset=&Add_Vector(
     A=<-22.1231 -11 -0.8408>
     B=&RotateVectorByQuaternion(Vector To Rotate=&RandomRangeVector(Minimum=<0 5 0>, Maximum=<30 0 2>), Quaternion=<0 -0.15 0 1>)
    )
    Sampled Position Offset Coordinate Space:ENiagaraMeshTransforms=::NewEnumerator3
    Skeletal Mesh=$System.Skeletal Mesh
   )
   &SkeletalMeshLocation(
    Bone / Socket Index=2
    Bone Sampling Mode=::NewEnumerator4
    Inherit Velocity=false
    MaskSpawnGroup=true
    Sampled Position Offset=$Particles.ShoulderSmokeOffset
    Sampled Position Offset Coordinate Space:ENiagaraMeshTransforms=::NewEnumerator3
    Skeletal Mesh=$User.SkeletalMesh
    Spawn Group Mask=1
   ) {module=SkeletalMeshLocation001}
   &AddVelocity(Velocity=<25 -50 50>)
   &InheritVelocity(Module.Limit Speed=false, Velocity Scale=&VectorFromFloat(Value=0.5))
   /SetVariables(Particles.SpawnGroup=&SpawnGroup()) {isSetParameters=true}
  ]
  particleUpdateScript [
   &ParticleState()
   &SkeletalMeshLocation(
    Apply To Particle Position=false
    Bone / Socket Index=2
    Bone Sampling Mode=::NewEnumerator4
    Inherit Velocity=false
    MaskSpawnGroup=false
    Sampled Position Offset=$Particles.ShoulderSmokeOffset
    Sampled Position Offset Coordinate Space:ENiagaraMeshTransforms=::NewEnumerator3
    Skeletal Mesh=$System.Skeletal Mesh
   ) {module=SkeletalMeshLocation003}
   /SetVariables(Particles.CompareSmokePosition:Vector3f=&Multiply_Vector(
     A=<0.4 0.4 1>
     B=&Subtract_Vector(A=$Output.SkeletalMeshLocation003.SampledPosition, B=Linked() {linkedVariable {name=Particles.Position, type=NiagaraPosition}})
    )) {isSetParameters=true}
   &Drag(Drag=5, Use Linear Drag=true, Use Rotational Drag=false)
   &GravityForce(Gravity=<0 0 100>)
   &InheritVelocity(
    Module.Limit Speed=true
    Velocity Scale=&Multiply_VectorByFloat(
     Vector=&RandomRangeVector(Minimum=<0.075 0.075 0.01>, Maximum=<0.125 0.125 0.01>)
     Float=&Multiply_Float(
      A=&Lerp_Float(A=1, B=0, Alpha=$Particles.NormalizedAge)
      B=&Lerp_Float(
       A=1
       B=0
       Alpha=&Multiply_Float(A=&VectorLengthSafe(Vector:Vector3f=$Particles.CompareSmokePosition), B=0.015)
      )
     )
    )
   ) {module=InheritVelocity001}
   &SolveForcesAndVelocity(Clamp Velocity=false, Limit Acceleration=false)
   &ScaleSpriteSize(Scale Factor=&Vector2DFromFloat(Value=&FloatFromCurve(FloatCurve=DataInterface() {propertyValues {Curve {keys [{arriveTangent=4.1867, interpMode=RCIM_Cubic, leaveTangent=4.1867, tangentMode=RCTM_User, time=0, value=0}, <at 0.35 1>]}}}))) {module=ScaleSpriteSize001}
   &ScaleSpriteSize(Scale Factor=&Vector2DFromFloat(Value=&FloatFromCurve(
      FloatCurve=DataInterface() {propertyValues {Curve {keys [
          {arriveTangent=-0.0248, interpMode=RCIM_Cubic, leaveTangent=-0.0248, tangentMode=RCTM_User, time=0, value=1}
          {arriveTangent=-0.5458, interpMode=RCIM_Cubic, leaveTangent=-0.5458, tangentMode=RCTM_User, time=1, value=0.75}
         ]}}}
      CurveIndex=&ClampFloat(Float=&Multiply_Float(A=&VectorLengthSafe(Vector:Vector3f=$Particles.CompareSmokePosition), B=0.015))
      Scale Curve=2
     )))
   /SetVariables(Particles.SubImageIndex=&RandomRangeFloat(Minimum=0, Maximum=63)) {isSetParameters=true}
   &DynamicMaterialParameters(
    Index 0 Param 1=&FloatFromCurve(FloatCurve=DataInterface() {propertyValues {Curve {keys [
         {arriveTangent=4.0913, interpMode=RCIM_Cubic, leaveTangent=4.0913, tangentMode=RCTM_User, time=0, value=0}
         {arriveTangent=0.1513, interpMode=RCIM_Cubic, leaveTangent=0.1513, tangentMode=RCTM_User, time=0.5, value=1}
         {arriveTangent=-0.0082, interpMode=RCIM_Cubic, leaveTangent=-0.0082, tangentMode=RCTM_User, time=1, value=0}
        ]}}})
    Index 0 Param 4=1
    Module.Param0WriteEnabled=true
    Module.Param1WriteEnabled=false
    Module.Param2WriteEnabled=false
    Module.Param3WriteEnabled=true
   )
   &ScaleColor(
    Scale Alpha=&FloatFromCurve(FloatCurve=DataInterface() {propertyValues {Curve {keys [
         {arriveTangent=-2.2302, interpMode=RCIM_Cubic, leaveTangent=-2.2302, tangentMode=RCTM_User, time=0, value=1}
         {arriveTangent=-0.022, interpMode=RCIM_Cubic, leaveTangent=-0.022, tangentMode=RCTM_User, time=1, value=0}
        ]}}})
    ScaleA=true
    ScaleRGB=false
   )
   &ScaleColor(
    Scale Alpha=&LerpMultipleFloats(
     Float Count:ENiagara_MultipleLerpCount=::NewEnumerator0
     1:NiagaraFloat=&RandomRangeFloat(Minimum=0.02, Maximum=0.04)
     2:NiagaraFloat=&RandomRangeFloat(Minimum=0.05, Maximum=0.1)
     3:NiagaraFloat=&RandomRangeFloat(Minimum=0.02, Maximum=0.04)
     4:NiagaraFloat=?
     5:NiagaraFloat=?
     Alpha:NiagaraFloat=Linked() {linkedVariable {name=NPC.FortniteNPC.TimeOfDay_0-1, type=NiagaraFloat}}
    )
    ScaleA=true
    ScaleRGB=false
   ) {module=ScaleColor002}
   &ScaleColor(
    Scale Alpha=&Lerp_Float(
     A=1
     B=0
     Alpha=&Multiply_Float(A=&VectorLengthSafe(Vector:Vector3f=$Particles.CompareSmokePosition), B=0.015)
    )
    ScaleA=true
    ScaleRGB=false
   ) {module=ScaleColor001}
   /SetVariables(Particles.StartFrame:NiagaraFloat=&RandomRangeFloat(Minimum=0, Maximum=63)) {isSetParameters=true}
   /SetVariables(Particles.SubImageIndex=&Lerp_Float(
     A=$Particles.StartFrame
     B=&Add_Float(A=$Particles.StartFrame, B=&RandomRangeFloat(Minimum=4, Maximum=8))
     Alpha=$Particles.NormalizedAge
    )) {isSetParameters=true}
   &Color(Color=&ColorFromCurve(
     ColorCurve=DataInterface() {propertyValues {
       AlphaCurve {keys [<at 0 1 RCIM_Cubic>]}
       BlueCurve {keys [<at 0 0 RCIM_Cubic>,<at 0.3333 0>,<at 0.6667 0>,<at 1 1 RCIM_Cubic>]}
       GreenCurve {keys [<at 0 0 RCIM_Cubic>,<at 0.3333 1>,<at 0.6667 1>,<at 1 0 RCIM_Cubic>]}
       RedCurve {keys [<at 0 1 RCIM_Cubic>,<at 0.3333 1>,<at 0.6667 0>,<at 1 0 RCIM_Cubic>]}
      }}
     CurveIndex=&Lerp_Float(
      A=1
      B=0
      Alpha=&Multiply_Float(A=&VectorLengthSafe(Vector:Vector3f=$Particles.CompareSmokePosition), B=0.015)
     )
    )) {enabled=false}
   &KillParticles(Kill Particles=&SetBoolByFloatComparison(
     A=&MakeFloatFromLinearColor(LinearColor=$Particles.Color, Channel:ENiagaraLinearColor_Channels=::NewEnumerator3)
     B=0.001
     Comparison Type=::NewEnumerator7
    ))
  ]
  rendererClasses [&NiagaraSpriteRendererProperties]
  renderers [{rendererClass=&NiagaraSpriteRendererProperties}]
 }
 SmokeRibbon {
  emitterUpdateScript [
   &EmitterState()
   &SpawnPerUnit(
    Module.Use Max Movement Threshold=false
    Module.Use Movement Tolerance=false
    Spawn Per Unit=15
    Spawn Probability:NiagaraFloat=&Lerp_Float(A=1, B=0, Alpha=&MakeFloatFromInt(Int=$System.isSkydiving))
    Use Spawn Probability=true
   )
   &SpawnRate(SpawnRate=20, Use Spawn Probability=false) {enabled=false}
  ]
  particleSpawnScript [
   /SetVariables(Particles.Lifetime=0.6, Particles.RibbonWidth=75) {isSetParameters=true}
   &SkeletalMeshLocation(
    Bone / Socket Index=2
    Bone Sampling Mode=::NewEnumerator4
    Inherit Velocity=false
    MaskSpawnGroup=false
    Sampled Position Offset=<-13.1117 -11 2.1285>
    Sampled Position Offset Coordinate Space:ENiagaraMeshTransforms=::NewEnumerator3
    Skeletal Mesh=$System.Skeletal Mesh
   )
   &AddVelocity(Velocity=<100 -200 100>)
   &AddVelocity(Velocity=<0 -1000 0>) {enabled=false, module=AddVelocity001}
   &InheritVelocity(
    Module.Limit Speed=false
    Velocity Scale=&VectorFromFloat(Value=&Lerp_Float(A=1, B=0, Alpha=$Particles.NormalizedAge))
   )
   /SetVariables(Particles.RibbonLinkOrder:NiagaraFloat=&MakeFloatFromInt(Int=/Niagara/DynamicInputs/DataInterface/GenerateIntFromCounter(SimpleCounter:NiagaraDataInterfaceSimpleCounter=DataInterface() {propertyValues {GpuSyncMode=None, InitialValue=0}}))) {isSetParameters=true}
   &SkeletalMeshLocation(
    Apply To Particle Position=false
    Bone Sampling Mode=::NewEnumerator4
    Inherit Velocity=false
    MaskSpawnGroup=false
    Sampled Position Offset=<-13.1117 -11 2.1285>
    Sampled Position Offset Coordinate Space:ENiagaraMeshTransforms=::NewEnumerator3
    Skeletal Mesh=$User.SkeletalMesh
   ) {enabled=false, module=SkeletalMeshLocation002}
   /SetVariables(
    Particles.MeshOrientationTest:Quat4f=$Output.SkeletalMeshLocation002.SampledRotation
    Particles.RibbonFacing=/Niagara/DynamicInputs/Helpers/SafeNormalizeVector(
     Vector To Normalize:Vector3f=&RotateVectorByQuaternion(Vector To Rotate=$Engine.Owner.Velocity, Quaternion=$Output.SkeletalMeshLocation002.SampledRotation)
     Fallback Vector:Vector3f=<0 0 0>
    )
   ) {enabled=false, isSetParameters=true}
   &SkeletalMeshLocation(
    Apply To Particle Position=false
    Bone / Socket Index=2
    Bone Sampling Mode=::NewEnumerator4
    Inherit Velocity=false
    MaskSpawnGroup=false
    Sampled Position Offset=<-13.1117 -11 2.1285>
    Sampled Position Offset Coordinate Space:ENiagaraMeshTransforms=::NewEnumerator3
    Skeletal Mesh=$System.Skeletal Mesh
   ) {module=SkeletalMeshLocation001}
   /SetVariables(
    Particles.MeshOrientationTest:Quat4f=$Output.SkeletalMeshLocation001.SampledRotation
    Particles.RibbonFacing=&RotateVectorByQuaternion(Vector To Rotate=<1 0 0>, Quaternion=$Output.SkeletalMeshLocation001.SampledRotation)
    Particles.RibbonFacingAlt:Vector3f=&RotateVectorByQuaternion(Vector To Rotate=<-1 0 0>, Quaternion=$Output.SkeletalMeshLocation001.SampledRotation)
    Particles.RibbonFacingAltAlt:Vector3f=&RotateVectorByQuaternion(Vector To Rotate=<0 0 -1>, Quaternion=$Output.SkeletalMeshLocation001.SampledRotation)
   ) {isSetParameters=true}
  ]
  particleUpdateScript [
   &ParticleState()
   &SkeletalMeshLocation(
    Apply To Particle Position:NiagaraBool=&SetBoolByFloatComparison(A=$Particles.NormalizedAge, B=0.25, Comparison Type=::NewEnumerator6)
    Bone / Socket Index=2
    Bone Sampling Mode=::NewEnumerator4
    Inherit Velocity=false
    MaskSpawnGroup=false
    Sampled Position Offset=<-13.1117 -11 2.1285>
    Sampled Position Offset Coordinate Space:ENiagaraMeshTransforms=::NewEnumerator3
    Skeletal Mesh=$System.Skeletal Mesh
   ) {module=SkeletalMeshLocation003}
   &Drag(Drag=5, Use Linear Drag=true, Use Rotational Drag=false)
   &GravityForce(Gravity=<0 0 600>)
   &SolveForcesAndVelocity(Clamp Velocity=false, Limit Acceleration=false)
   /Game/Effects/Niagara/Modules/Materials/FN_RibbonMaterial/FN_RibbonMaterial_v2(
    Color:LinearColor=&MakeLinearColorFromVectorAndFloat(
     "Vector (RGB)"=<1 1 1>
     "Float (Alpha)"=&FloatFromCurve(
      FloatCurve=DataInterface() {propertyValues {Curve {keys [
          {arriveTangent=0.0082, interpMode=RCIM_Cubic, leaveTangent=0.0082, tangentMode=RCTM_User, time=0.25, value=0.2}
          {arriveTangent=-0.4528, interpMode=RCIM_Cubic, leaveTangent=-0.4528, tangentMode=RCTM_User, time=1, value=0}
         ]}}}
      Scale Curve=&LerpMultipleFloats(
       Float Count:ENiagara_MultipleLerpCount=::NewEnumerator0
       1:NiagaraFloat=0.6
       2:NiagaraFloat=1
       3:NiagaraFloat=0.6
       4:NiagaraFloat=?
       5:NiagaraFloat=?
       Alpha:NiagaraFloat=Linked() {linkedVariable {name=NPC.FortniteNPC.TimeOfDay_0-1, type=NiagaraFloat}}
      )
     )
    )
    Color Pan U Speed=2.8
    Color Scale U=2
    Erosion Amount:NiagaraFloat=&Lerp_Float(A=0, B=1.75, Alpha=$Particles.NormalizedAge)
    Erosion Pan Speed U=2.75
    Erosion Power=0.703
    Erossion Scale U=2
    Opacity Pan U Speed=0.23
    Opacity Scale U=0.2
    Scale Alpha:NiagaraFloat=&FloatFromCurve(
     FloatCurve=DataInterface() {propertyValues {Curve {keys [
         {interpMode=RCIM_Cubic, tangentMode=RCTM_User, time=0, value=1}
         {arriveTangent=-2.8707, interpMode=RCIM_Cubic, leaveTangent=-2.8707, tangentMode=RCTM_User, time=1, value=0}
        ]}}}
     Scale Curve=2
    )
    TIme:NiagaraFloat=$Engine.Time
   )
   &ScaleRibbonWidth(Ribbon Width Scale=&FloatFromCurve(FloatCurve=DataInterface() {propertyValues {Curve {keys [
         {arriveTangent=0.4163, interpMode=RCIM_Cubic, leaveTangent=0.4163, tangentMode=RCTM_User, time=0.3, value=1}
         {arriveTangent=-1.9468, interpMode=RCIM_Cubic, leaveTangent=-1.9468, tangentMode=RCTM_User, time=1.0032, value=0.0313}
        ]}}}))
   &ScaleColor(Scale Alpha=0.25, ScaleA=true, ScaleRGB=true)
  ]
  rendererClasses [&NiagaraRibbonRendererProperties, &NiagaraMeshRendererProperties]
  renderers [
   {rendererClass=&NiagaraRibbonRendererProperties}
   {rendererClass=&NiagaraMeshRendererProperties, rendererIndex=1}
   {rendererClass=&NiagaraRibbonRendererProperties, rendererIndex=2}
   {rendererClass=&NiagaraRibbonRendererProperties, rendererIndex=3}
  ]
 }
}
summary {
 systemName=NS_PieMan_Skin_Original
 emitters [
  {emitterName=DroolSource, rendererClasses [&NiagaraSpriteRendererProperties]}
  {emitterName=Drool, rendererClasses [&NiagaraMeshRendererProperties, &NiagaraSpriteRendererProperties]}
  {emitterName=Smoke, rendererClasses [&NiagaraSpriteRendererProperties]}
  {emitterName=SmokeRibbon, rendererClasses [&NiagaraRibbonRendererProperties, &NiagaraMeshRendererProperties]}
  {emitterName=Flies_Optimized, rendererClasses [&NiagaraMeshRendererProperties]}
  {emitterName=Fly_Trail, rendererClasses [&NiagaraRibbonRendererProperties]}
  {emitterName=DroolSimple, rendererClasses [&NiagaraSpriteRendererProperties, &NiagaraMeshRendererProperties]}
  {bEnabled=false, emitterName=DEBUG_Sourcing, rendererClasses [&NiagaraSpriteRendererProperties]}
 ]
 userVariables [
  User.bisFootHidden {type=NiagaraBool}
  User.bIsFrontEnd {type=NiagaraBool}
  User.bIsFrontEndPreview {type=NiagaraBool}
  User.bisHandHidden {type=NiagaraBool}
  User.bIsJumping {type=NiagaraBool}
  User.IsSkydiving {type=NiagaraBool}
  User.JumpTime {type=NiagaraFloat, defaultValue=0, description=""}
  User.RibbonIDOffset {type=NiagaraInt32}
  User.SkeletalMesh {
   type=NiagaraDataInterfaceSkeletalMesh
   defaultValue {
    struct=NiagaraExt_VariableValue_DataInterface
    value {dataInterface=&self:NiagaraDataInterfaceSkeletalMesh_0}
   }
  }
  User.SpawnLandedFX {type=NiagaraBool}
 ]
}
